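# Translucent unlit speedline material for Niagara ribbons in the Pickaxe_Stamina_Male weapon effects.
# Emissive is particle color times a vector parameter, passed through MF_FX_EmissiveAdjustments.
# Opacity comes from the speedline texture's red channel, scaled by a parameter, particle alpha and a SphereMask along U.
# An optional static switch breaks up the line with panning, step-thresholded noise and a smoothstep edge mask along U.
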Material /BRCosmetics/Weapons/FORT_Melee/Pickaxe_Stamina_Male/FX/M_Speedline
graph {
 properties {
  bEnableResponsiveAA=true
  bUsedWithNiagaraRibbons=true
  blendMode=BLEND_Translucent
  pixelDepthOffsetMode=PDOM_Legacy
  shadingModel=MSM_Unlit
  usedShadingModels=Unlit
 }
 nodes [
  Constant()
  n1 = ParticleColor()
  MaterialFunctionCall(Emissive=Multiply(A=VectorParameter() {parameterName=ParticleColorMult, defaultValue=<rgba 0 0 0 1>}.RGB, B=@n1.RGB)) {materialFunction=&MF_FX_EmissiveAdjustments}
  n2 = TextureSampleParameter2D!(UVs, Level, Apply View MipBias) {
   mipValueMode=TMVM_MipLevel
   parameterName=Texture
   samplerType=SAMPLERTYPE_LinearColor
   texture=&/T_Speedline
  }
  n3 = Multiply(A=@n2.R, B=ScalarParameter() {defaultValue=45, parameterName=B})
  n4 = TextureCoordinate() {uTiling=0.5, vTiling=0.5}
  n5 = AppendVector(
   A=ScalarParameter() {defaultValue=0.25, group=PanningNoise, parameterName=PanSpeed_X}
   B=ScalarParameter() {defaultValue=0.5, group=PanningNoise, parameterName=PanSpeed_Y}
  )
  n6 = Step(Y=TextureSampleParameter2D!(UVs=Panner(Coordinate=@n4, Speed=@n5), Apply View MipBias) {group=PanningNoise, parameterName=TextureInput, texture=/Game/Effects/FogVolume/T_TileNoise}.RGB) {constX=0.3}
  n7 = ScalarParameter() {defaultValue=0.2, group=PanningNoise, parameterName=PanSpeedEdgeMask}
  n8 = ComponentMask(None=TextureCoordinate()) {r=true}
  n9 = Multiply(A=@n3, B=Multiply(A=Saturate(None=@n6), B=SmoothStep(Min=@n7, Value=@n8) {constMin=0.2}))
  n10 = ComponentMask(None=TextureCoordinate()) {r=true}
  n11 = ScalarParameter() {parameterName=SphereMAsk}
  Multiply(
   A=@n1.A
   B=Multiply(
    A=StaticSwitchParameter(True=@n9, False=@n3) {group=PanningNoise, parameterName=UsePanningNoise?}
    B=SphereMask(A=@n10, B=@n11) {attenuationRadius=1, hardnessPercent=0}
   )
  )
 ]
}
